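# Material function that paints a fake glowing Nexus core over incoming material attributes, overriding Emissive and Specular within a radial mask.
# It projects cracked-glass textures on camera-facing virtual planes around a center point, distorting radial falloffs for dark mask, color gradient and glow layers.
# Layered sines flicker the inner glow, with MPC_Nexus Destabilize raising speed, intensity and tint.
# Forward and mobile paths use one crack sample; a second output exposes unflickered color.
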MaterialFunction /Game/Athena/Environments/Nexus/Materials/MF_NexCoreGlowSimple
graph {nodes [
  n1 = GetMaterialAttributes(None=FunctionInput() {inputType=FunctionInput_MaterialAttributes}) {attributeGetTypes [#1, #2]}
  n2 = ScalarParameter() {defaultValue=1, parameterName=NexusCoreEmissiveBoost}
  n3 = Constant()
  n4 = FunctionInput(Preview=VectorParameter() {parameterName=SphereCenter, defaultValue=<rgba -3024 -1776 96 1>}.RGB) {bUsePreviewValueAsDefault=true, inputName=Centerlocation, sortPriority=2}
  n5 = Normalize(VectorInput=Subtract(A=@n4, B=CameraPositionWS()))
  n6 = VectorParameter() {parameterName=Axis, defaultValue=<rgba 0 0 1 1>}
  n7 = ScalarParameter() {defaultValue=-1000, parameterName=OffsetPlaneCore}
  n8 = ScalarParameter() {defaultValue=0.8, parameterName=FlickerSpeed}
  n9 = Multiply(
   A=Time()
   B=LinearInterpolate(
    A=@n8
    B=Multiply(A=@n8) {constB=2}
    Alpha=CollectionParameter() {collection=/Game/Athena/Environments/Nexus/Materials/FXshader/MPC_Nexus, parameterName=Destabilize}
   )
  )
  n10 = Add(A=Multiply(B=Sine(None=@n9) {period=0.55}), B=Multiply(A=Sine(None=@n9) {period=0.11}) {constB=0.2})
  n11 = Multiply(A=Sine(None=@n9) {period=0.175}) {constB=0.3}
  n12 = Multiply(A=Sine(None=@n9) {period=3.25}, B=Sine(None=@n9) {period=2.325})
  n13 = Multiply(
   A=Multiply(A=Add(A=@n10, B=@n11), B=ConstantBiasScale(None=@n12))
   B=LinearInterpolate(Alpha=CollectionParameter() {collection=&MPC_Nexus, parameterName=Destabilize}) {constA=0.5, constB=1.2}
  )
  n14 = ScalarParameter() {defaultValue=0.07, parameterName=FlickerMult}
  n15 = StaticSwitchParameter(
   True=Add(A=ScalarParameter() {defaultValue=1200, parameterName=ObjectScale}, B=Multiply(A=@n13, B=@n14) {constB=0.05})
   False=MaterialFunctionCall() {materialFunction=/Engine/Functions/Engine_MaterialFunctions02/WorldPositionOffset/ObjectScale}.Scale Z
  ) {parameterName=MeshParticle}
  n16 = StaticSwitchParameter(True=Multiply(A=@n7, B=@n15), False=@n7) {parameterName=UseObjectScale}
  n17 = Normalize(VectorInput=Subtract(A=@n4, B=CameraPositionWS()))
  n18 = ScalarParameter() {defaultValue=1185.5463, parameterName=SizeProjectionTextureCore}
  n19 = MaterialFunctionCall!(
   Plane Normal Axis=@n5
   Plane X Axis=Normalize(VectorInput=CrossProduct(A=@n5, B=@n6.RGB))
   Plane Center=Add(A=@n4, B=Multiply(A=@n16, B=@n17))
   UV Scale=StaticSwitchParameter(True=Multiply(A=@n18, B=@n15), False=@n18) {parameterName=UseObjectScale}
  ) {materialFunction=&/VirtualPlaneCoordinatesAbs}
  n20 = TextureSample(UVs=@n19.UVs) {texture=/Game/Effects/Fort_Effects/Textures/Tile/T_Glass_Cracks_M}
  n21 = VectorParameter() {parameterName=Axis, defaultValue=<rgba 0 0 1 1>}
  n22 = ScalarParameter() {defaultValue=500, parameterName=OffsetCenter2Core}
  n23 = StaticSwitchParameter(True=Multiply(A=@n22, B=@n15), False=@n22) {parameterName=UseObjectScale}
  n24 = ScalarParameter() {defaultValue=1063.7999, parameterName=SizeProjectionTexture2Core}
  n25 = MaterialFunctionCall(
   Plane Normal Axis=@n5
   Plane X Axis=Normalize(VectorInput=CrossProduct(A=@n5, B=@n21.RGB))
   Plane Center=Add(A=@n4, B=Multiply(A=@n17, B=@n23))
   UV Scale=StaticSwitchParameter(True=Multiply(A=@n24, B=@n15), False=@n24) {parameterName=UseObjectScale}
  ) {materialFunction=&VirtualPlaneCoordinatesAbs}
  n26 = Add(A=@n20.R, B=TextureSample(UVs=Add(A=@n25.UVs) {constB=0.5}) {texture=&T_Glass_Cracks_M}.R)
  n27 = ShadingPathSwitch(Default=@n26, Deferred=@n26, Forward=@n20.R, Mobile=@n20.R)
  n28 = QualitySwitch(Default=@n27, High=@n27, Medium=@n27)
  n29 = ScalarParameter() {defaultValue=-0.1238, parameterName=Distortion}
  n30 = Constant2Vector() {g=0.5, r=0.5}
  n31 = Fresnel(ExponentIn=ScalarParameter() {defaultValue=0.4967, parameterName=DistortionExp}) {baseReflectFraction=0}
  n32 = ScalarParameter() {defaultValue=0.2744, parameterName=DistortionNormal}
  n33 = Multiply(A=Add(A=Multiply(A=@n28, B=@n29), B=Distance(A=@n19.UVs, B=@n30)), B=OneMinus(None=Multiply(A=@n31, B=@n32)))
  n34 = Divide(A=@n33, B=ScalarParameter() {defaultValue=0.7709, parameterName=DistanceFalloffBlackmask})
  n35 = ScalarParameter() {defaultValue=20.1444, parameterName=FresnelEXPBlackMask}
  n36 = CameraVectorWS()
  n37 = Ceil(None=Saturate(None=DotProduct(A=@n5, B=@n36)))
  n38 = Clamp(None=Add(A=Power(Base=@n34, Exp=@n35), B=@n37))
  n39 = OneMinus(None=@n37)
  n40 = Power(
   Base=Divide(A=@n33, B=ScalarParameter() {defaultValue=0.9241, parameterName=DistanceFalloff})
   Exp=ScalarParameter() {defaultValue=0.3703, parameterName=FresnelEXP}
  )
  n41 = VectorParameter() {parameterName=CoreGLow, defaultValue=<rgba 1.6276 2.7893 4 1>}
  n42 = ScalarParameter() {defaultValue=0.4667, parameterName=Glow}
  n43 = ScalarParameter() {defaultValue=0.9691, parameterName=DistanceFalloff2}
  n44 = Power(Base=OneMinus(None=Divide(A=@n33, B=@n43)), Exp=ScalarParameter() {defaultValue=8, parameterName=FresnelEXP3})
  n45 = Add(
   A=LinearInterpolate(
    A=VectorParameter() {parameterName=EmColor1, defaultValue=<rgba 0.1636 0.4979 0.9948 1>}.RGB
    B=VectorParameter() {parameterName=EmColor2, defaultValue=<rgba 0 0 0 1>}.RGB
    Alpha=Clamp(None=@n40)
   )
   B=Multiply(A=Multiply(A=@n41.RGB, B=@n42), B=Clamp(None=@n44))
  )
  n46 = VectorParameter() {parameterName=CoreGLow2, defaultValue=<rgba 0.8339 1.6354 2 1>}
  n47 = LinearInterpolate(
   A=@n46.RGB
   B=Multiply(A=VectorParameter() {parameterName=DestabilizeTint, defaultValue=<rgba 10 0.2 6.3603 0>}.RGB, B=@n46.RGB)
   Alpha=CollectionParameter() {collection=&MPC_Nexus, parameterName=Destabilize}
  )
  n48 = Divide(A=@n33, B=ScalarParameter() {defaultValue=0.7881, parameterName=DistanceFalloff3})
  n49 = Clamp(None=Power(Base=OneMinus(None=@n48), Exp=ScalarParameter() {defaultValue=2.8268, parameterName=FresnelEXP4}))
  n50 = Add(A=@n45, B=Multiply(A=LinearInterpolate(Alpha=@n13) {constA=0.4}, B=Multiply(A=@n47, B=@n49)))
  n51 = Multiply(A=ParticleColor().RGB, B=Add(A=LinearInterpolate(A=@n3, B=@n1.EmissiveColor, Alpha=@n38), B=Multiply(A=@n39, B=@n50)) {constB=0})
  FunctionOutput(None=SetMaterialAttributes!(
    MaterialAttributes=@n1.MaterialAttributes
    EmissiveColor=Multiply(A=@n2, B=@n51)
    Specular=LinearInterpolate(A=@n3, B=@n1.Specular, Alpha=@n38)
   ) {attributeSetTypes [#1, #2]})
  FunctionOutput(None=@n45) {outputName=Color}
 ]}
