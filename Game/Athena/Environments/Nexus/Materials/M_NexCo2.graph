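# M_NexCo2 is a Default Lit material for the Nexus environment, usable on mesh particles, that defaults to a black surface with a pale Fresnel rim glow.
# Color, specular and roughness parameters define the surface, while a Fresnel blend between two colors, scaled by intensity, drives emissive.
# These attributes pass through MF_NexCoreGlowSimple, whose center location static switches select from object position, pivot point or actor position.
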Material /Game/Athena/Environments/Nexus/Materials/M_NexCo2
graph {
 properties {bUseMaterialAttributes=true, bUsedWithMeshParticles=true, pixelDepthOffsetMode=PDOM_Legacy, usedShadingModels=Default Lit}
 nodes [
  n1 = ScalarParameter() {defaultValue=1, parameterName=Edge Fresnel Emissive Intensity}
  n2 = LinearInterpolate(
   A=VectorParameter() {parameterName=EmissiveFresnel1, defaultValue=<rgba 0 0 0 1>}.RGB
   B=VectorParameter() {parameterName=EmissiveFresnel2, defaultValue=<rgba 0.9077 0.8385 1 1>}.RGB
   Alpha=Fresnel(ExponentIn=ScalarParameter() {defaultValue=10.2867, parameterName=Exp}) {baseReflectFraction=0}
  )
  n3 = MaterialFunctionCall() {materialFunction=&ObjectPivotPoint}
  n4 = ActorPositionWS()
  MaterialFunctionCall!(
   In=MakeMaterialAttributes!(
    BaseColor=VectorParameter() {parameterName=Color, defaultValue=<rgba 0 0 0 1>}.RGB
    Metallic
    Specular=ScalarParameter() {defaultValue=0.1143, parameterName=Specular}
    Roughness=ScalarParameter() {parameterName=Roughness}
    Anisotropy
    EmissiveColor=Multiply(A=@n1, B=@n2)
    Opacity
    OpacityMask
    Normal
    Tangent
    WorldPositionOffset
    SubsurfaceColor
    ClearCoat
    ClearCoatRoughness
    AmbientOcclusion
    Refraction
    CustomizedUVs_0
    CustomizedUVs_1
    CustomizedUVs_2
    CustomizedUVs_3
    CustomizedUVs_4
    CustomizedUVs_5
    CustomizedUVs_6
    CustomizedUVs_7
    PixelDepthOffset
    ShadingModel
    Displacement
   )
   Centerlocation=StaticSwitchParameter(
    True=ObjectPositionWS()
    False=StaticSwitchParameter(True=@n3.Object Pivot Location, False=@n4) {parameterName=UseObjectPivotInsteadofActorPostion}
   ) {parameterName=MeshParticle}
  ) {materialFunction=&/MF_NexCoreGlowSimple}
 ]
}
