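# Additive, unlit, two-sided mesh-particle material that draws glowing cracks and light rays for the Nexus core.
# Vertex blue separates cracks from rays, selecting color, noise-driven opacity and depth fade on rays.
# World position offset pushes rays outward from the pivot via UV1 and dynamic WPOMult, plus layered-sine vibration driven by MPC_Nexus Destabilize.
# Static switches add camera-distance scale-down, time-of-day opacity and near-camera fade.
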Material /Game/Athena/Environments/Nexus/Materials/M_Nexus_Core_Crack
graph {
 properties {
  bUsedWithMeshParticles=true
  blendMode=BLEND_Additive
  pixelDepthOffsetMode=PDOM_Legacy
  previewMesh=/Game/Effects/Fort_Effects/Effects/Environment/DarkTower/SM_DarkTower_ZPContainment_Rays
  shadingModel=MSM_Unlit
  twoSided=true
  usedShadingModels=Unlit
 }
 nodes [
  n1 = Normalize(VectorInput=Subtract(
    A=WorldPosition().XYZ
    B=MaterialFunctionCall() {materialFunction=&ObjectPivotPoint}.Object Pivot Location
   ))
  n2 = ScalarParameter() {defaultValue=0.8, parameterName=FlickerSpeed}
  n3 = Multiply(
   A=Time()
   B=LinearInterpolate(
    A=@n2
    B=Multiply(A=@n2) {constB=2}
    Alpha=CollectionParameter() {collection=/Game/Athena/Environments/Nexus/Materials/FXshader/MPC_Nexus, parameterName=Destabilize}
   )
  )
  n4 = Multiply(B=Sine(None=@n3) {period=0.55})
  n5 = Multiply(A=Sine(None=@n3) {period=0.11}) {constB=0.2}
  n6 = Sine(None=@n3) {period=0.175}
  n7 = Multiply(A=Sine(None=@n3) {period=3.25}, B=Sine(None=@n3) {period=2.325})
  n8 = Multiply(A=Add(A=Add(A=@n4, B=@n5), B=Multiply(A=@n6) {constB=0.3}), B=@n7)
  n9 = CollectionParameter() {collection=&MPC_Nexus, parameterName=Destabilize}
  n10 = Multiply(A=@n1, B=Multiply(A=Multiply(A=@n8, B=@n9), B=ScalarParameter() {parameterName=WPOVibrate}))
  n11 = TextureCoordinate() {coordinateIndex=1}
  n12 = Multiply(
   A=LinearInterpolate(
    A=ScalarParameter() {parameterName=PushDistanceMin}
    B=ScalarParameter() {defaultValue=2500, parameterName=PushDistance}
    Alpha=ComponentMask(None=@n11) {g=true}
   )
   B=MaterialFunctionCall() {materialFunction=/Engine/Functions/Engine_MaterialFunctions02/WorldPositionOffset/ObjectScale}.Scale Z
  )
  n13 = VertexColor()
  n14 = Multiply(
   A=LinearInterpolate(A=@n10, B=Multiply(A=@n12, B=@n1), Alpha=@n13.B)
   B=DynamicParameter() {paramNames [WPOMult,Param2,Param3,Param4]}.WPOMult
  )
  Add(A=@n14, B=@n10)
  Constant3Vector()
  ConstantBiasScale(None=@n7)
  n15 = ParticleColor()
  n16 = Multiply(
   A=Multiply(
    A=VectorParameter() {parameterName=Color, defaultValue=<rgba 0.3727 0.4306 1 1>}.RGB
    B=ScalarParameter() {defaultValue=1993.9108, parameterName=GlowCracks}
   )
   B=Power(Base=VertexColor().R, Exp=ScalarParameter() {defaultValue=4.3204, parameterName=OpacityCracksFalloff})
  )
  n17 = TextureCoordinate() {coordinateIndex=1}
  n18 = OneMinus(None=ComponentMask(None=@n17) {g=true})
  n19 = LinearInterpolate(
   A=VectorParameter() {parameterName=ColorRays, defaultValue=<rgba 0.4891 0.3203 1 1>}.RGB
   B=VectorParameter() {parameterName=ColorRays2, defaultValue=<rgba 0.1599 0.7576 1 1>}.RGB
   Alpha=@n18
  )
  n20 = VertexColor()
  MaterialFunctionCall(Emissive=Multiply(A=@n15.RGB, B=LinearInterpolate(A=@n16, B=@n19, Alpha=@n20.B))) {materialFunction=&MF_FX_EmissiveAdjustments}
  n21 = SphereMask(
   A=CameraPositionWS()
   B=ObjectPositionWS()
   Radius=ScalarParameter() {defaultValue=3500, group=Scale Down By Distance, parameterName=ScaleDownRadius}
   Hardness=Constant()
  )
  n22 = ScalarParameter() {defaultValue=0.25, group=Scale Down By Distance, parameterName=ScaleDownRadiusPower}
  StaticSwitchParameter(True=LinearInterpolate(A=@n14, Alpha=Power(Base=@n21, Exp=@n22)) {constB=0}, False=@n14) {parameterName=ScaleDownByDistance}
  n23 = ScalarParameter() {defaultValue=0.005, parameterName=MinCrackOpacity}
  n24 = TextureCoordinate()
  n25 = ScalarParameter() {defaultValue=2, parameterName=TilingNoise}
  n26 = ScalarParameter() {defaultValue=0.1, parameterName=Pan1}
  n27 = Constant()
  n28 = TextureSample(UVs=Panner(Coordinate=Multiply(A=@n24, B=@n25), Speed=AppendVector(A=@n26, B=@n27))) {samplerType=SAMPLERTYPE_Grayscale, texture=/Game/Global/Textures/Default/Noise/TilingNoise06}
  n29 = Multiply(A=ScalarParameter() {defaultValue=1, parameterName=OpacityCracks}, B=Saturate(None=Add(A=@n23, B=@n28.G)))
  n30 = ScalarParameter() {defaultValue=7, parameterName=RayBottomPowerMult}
  n31 = ScalarParameter() {defaultValue=3, parameterName=OpacityRaysFalloff}
  n32 = Multiply(A=Power(Base=@n18, Exp=Multiply(A=@n30, B=@n31) {constB=7}), B=ScalarParameter() {defaultValue=5, parameterName=RayBottomMult}) {constB=5}
  n33 = Multiply(A=Power(Base=@n18, Exp=@n31)) {constB=0.5}
  n34 = Saturate(None=ComponentMask(None=@n17) {r=true})
  n35 = ScalarParameter() {defaultValue=0.3, parameterName=OpacityRays}
  n36 = Multiply(A=@n28.G, B=Multiply(A=Add(A=@n32, B=@n33), B=Multiply(A=@n34, B=@n35)))
  n37 = LinearInterpolate(A=ScalarParameter() {defaultValue=500, parameterName=FadeDistance}, Alpha=@n18) {constB=0}
  n38 = LinearInterpolate(B=ShadingPathSwitch(Default=DepthFade(FadeDistance=@n37), Mobile=Constant() {r=1}) {desc=Migrated from FeatureLevelSwitch}, Alpha=@n13.B) {constA=1}
  n39 = TextureSample(UVs=TextureCoordinate() {coordinateIndex=2}) {texture=/Game/Global/Textures/Default/Masks_and_Gradients/T_Ramp_Virt}
  n40 = Multiply(
   A=Multiply(A=@n15.A, B=LinearInterpolate(A=@n29, B=@n36, Alpha=@n20.B))
   B=LinearInterpolate(A=@n38, B=Multiply(A=@n38, B=@n39.B), Alpha=ScalarParameter() {parameterName=UseVertexMask})
  )
  n41 = ScalarParameter() {defaultValue=0.26, parameterName=ToDScale_Morning}
  n42 = ScalarParameter() {defaultValue=1, parameterName=ToDScale_Day}
  n43 = ScalarParameter() {defaultValue=0.26, parameterName=ToDScale_Evening}
  n44 = ScalarParameter() {defaultValue=0.26, parameterName=ToDScale_Night}
  n45 = StaticSwitchParameter(
   True=MaterialFunctionCall(
    Color_Morning=Multiply(A=@n40, B=@n41) {constA=1}
    Color_Day=Multiply(A=@n40, B=@n42) {constA=1}
    Color_Evening=Multiply(A=@n40, B=@n43) {constA=1}
    Color_Night=Multiply(A=@n40, B=@n44) {constA=1}
   ) {materialFunction=&Color_Time_of_Day}.Result
   False=@n40
  ) {parameterName=Use color time of day}
  n46 = ScalarParameter() {defaultValue=250, parameterName=NearCamFade_Distance}
  StaticSwitchParameter(
   True=Multiply(A=@n45, B=MaterialFunctionCall(Distance=@n46) {materialFunction=&Near_Camera_Fade}.Result)
   False=@n45
  ) {parameterName=Use_NearCamFade?}
 ]
}
