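# Translucent, unlit, two-sided material for Zero Point spline meshes in the Nexus environment, usable on Niagara mesh particles.
# Emissive blends two colors by vertex color over panning magnetic-wave and caustic textures, shifts toward a core color via sphere mask, then boosts to HDR.
# Opacity combines panning noise, edge, camera, depth and radial fades.
# SplineThicken sets width, swollen by a collection-driven radial pulse; default shading path adds sine wobble.
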Material /Game/Athena/Environments/Nexus/Materials/M_ZP_SplineMeshes
graph {
 properties {
  bUsedWithNiagaraMeshParticles=true
  blendMode=BLEND_Translucent
  pixelDepthOffsetMode=PDOM_Legacy
  previewMesh=/Game/Athena/Environments/Nexus/Meshes/SM_ZeroPoint_SplineMeshes_01
  shadingModel=MSM_Unlit
  twoSided=true
  usedShadingModels=Unlit
 }
 nodes [
  MaterialFunctionCall() {materialFunction=&DebugTimeSine}
  n1 = VectorParameter() {parameterName=Colo_02, defaultValue=<rgba 0 0.3622 1 1>}
  n2 = LinearInterpolate(
   A=VectorParameter() {parameterName=Colo_01, defaultValue=<rgba 0.7608 0 1 1>}.RGB
   B=@n1.RGB
   Alpha=Power(Base=VertexColor().R) {constExponent=0.65}
  )
  n3 = VectorParameter() {parameterName=Tile1, defaultValue=<rgba 0.4 0.5 0 0>}
  n4 = AppendVector(A=@n3.R, B=@n3.R)
  n5 = VectorParameter() {parameterName=Tile1SqueezeMult, defaultValue=<rgba 0.3 2.5 0 1>}
  n6 = DynamicParameter() {defaultValue=<rgba 1 1 12 0>, paramNames [Random, WPO Gain Multiplier, Unassigned, Squeeze]}
  n7 = LinearInterpolate(A=@n4, B=Multiply(A=@n4, B=AppendVector(A=@n5.R, B=@n5.R)), Alpha=@n6.Squeeze) {constA=0.8, constB=2}
  n8 = Multiply(A=Add(A=TextureCoordinate() {coordinateIndex=1}, B=Multiply(B=@n7) {constA=-0.2, constB=0.5}), B=@n7)
  n9 = Multiply(A=@n8) {constB=1.5}
  n10 = VectorParameter() {parameterName=PanDirection1, defaultValue=<rgba 0.2 0.02 0 0>}
  n11 = AppendVector(A=@n10.R, B=@n10.R)
  n12 = Add(A=Multiply(A=Time(), B=ScalarParameter() {defaultValue=0.25, parameterName=PanSpeed}), B=@n6.Random)
  n13 = VectorParameter() {parameterName=PanDirection1Mult, defaultValue=<rgba 1 -1 0 1>}
  n14 = Multiply(A=@n11, B=AppendVector(A=@n13.R, B=@n13.R))
  n15 = Add(
   A=TextureSample(UVs=Panner(Coordinate=@n9, Time=@n12, Speed=@n11) {speedX=-0.2, speedY=-0.02}) {samplerSource=SSM_Wrap_WorldGroupSettings, texture=/Game/Athena/Prototype/Blueprints/Sneezy/Textures/T_Sneezy_MagneticWaves}.RGB
   B=TextureSample(UVs=Panner(Coordinate=@n8, Time=@n12, Speed=@n14) {speedX=-0.2, speedY=0.02}) {samplerSource=SSM_Wrap_WorldGroupSettings, texture=&T_Sneezy_MagneticWaves}.RGB
  )
  n16 = VectorParameter() {parameterName=Tile2, defaultValue=<rgba 0.3 0.5 0 0>}
  n17 = VectorParameter() {parameterName=PanDirection2, defaultValue=<rgba 0.1 0.04 0 0>}
  n18 = Panner(Coordinate=Multiply(A=TextureCoordinate(), B=AppendVector(A=@n16.R, B=@n16.R)), Time=@n12, Speed=AppendVector(A=@n17.R, B=@n17.R)) {speedX=0.1, speedY=0.04}
  n19 = LinearInterpolate(
   A=TextureSample(UVs=Add(A=@n18) {constB=0.01}) {samplerSource=SSM_Wrap_WorldGroupSettings, texture=/Game/Global/Textures/Default/Caustics/T_Caustics_Softer}.G
   B=TextureSample(UVs=@n18) {samplerSource=SSM_Wrap_WorldGroupSettings, texture=&T_Caustics_Softer}.G
  )
  n20 = Constant() {r=0.5}
  n21 = SphereMask(A=TextureCoordinate(), B=@n20, Radius=@n20) {hardnessPercent=0}
  n22 = Multiply(A=ParticleColor().RGB, B=Add(A=Add(A=@n15, B=@n19), B=Power(Base=@n21)))
  n23 = SphereMask(
   A=TransformPosition(None=WorldPosition().XYZ) {transformSourceType=TRANSFORMPOSSOURCE_World}
   B=Constant3Vector()
   Radius=ScalarParameter() {defaultValue=800, parameterName=Core Color Falloff Radius}
   Hardness=Constant()
  )
  n24 = Clamp(None=LinearInterpolate(A=Multiply(A=@n2, B=@n22), B=Multiply(A=@n1.RGB) {constB=6}, Alpha=Clamp(None=@n23)))
  MaterialFunctionCall(Emissive=MaterialFunctionCall(LDR Input=@n24, HDR Tint=@n24, HDR Intensity Multiplier=ScalarParameter() {defaultValue=9, parameterName=HDR Intensity}) {materialFunction=&DeriveHDRfromLDR}.RemappedRange) {materialFunction=&MF_FX_EmissiveAdjustments}
  n25 = Power(Base=VertexColor().R, Exp=ScalarParameter() {defaultValue=4, parameterName=ScaleByVertColorPower}) {constExponent=0.25}
  n26 = TextureCoordinate()
  n27 = ComponentMask(None=VectorParameter() {defaultValue=<rgba 0.2 5 0 1>}.RGB) {g=true, r=true}
  n28 = TextureSample(UVs=Panner(Coordinate=Multiply(A=@n26, B=@n27)) {speedY=0.2}) {samplerSource=SSM_Wrap_WorldGroupSettings, samplerType=SAMPLERTYPE_Grayscale, texture=&LowResBlurredNoise}
  n29 = ComponentMask(None=TextureCoordinate()) {r=true}
  n30 = Multiply(A=OneMinus(None=@n29), B=@n29)
  n31 = Multiply(A=@n30, B=@n30)
  n32 = Clamp(None=Multiply(A=Multiply(A=@n28.G, B=@n31)) {constB=15})
  n33 = ScalarParameter() {defaultValue=4096, parameterName=Fade Near Distance}
  n34 = Multiply(
   A=Clamp(None=Add(A=@n25) {constB=0.325})
   B=Multiply(A=@n32, B=MaterialFunctionCall(Distance=@n33) {materialFunction=&Near_Camera_Fade}.Result)
  )
  n35 = TransformPosition(None=WorldPosition().XYZ) {transformSourceType=TRANSFORMPOSSOURCE_World}
  n36 = SphereMask(
   A=@n35
   B=Constant3Vector()
   Radius=ScalarParameter() {defaultValue=850, parameterName=Depth Fade Falloff Radius}
   Hardness=Constant() {r=0.75}
  )
  n37 = QualitySwitch(Default=DepthFade(Opacity=@n34, FadeDistance=LinearInterpolate(Alpha=@n36) {constA=2048, constB=64}) {fadeDistanceDefault=480}, Low=@n34)
  n38 = ComponentMask(None=TextureCoordinate()) {g=true}
  n39 = ScalarParameter() {defaultValue=1, parameterName=Power_LengthMask}
  n40 = StaticSwitchParameter(True=Multiply(A=@n37, B=Power(Base=@n38, Exp=@n39)), False=@n37) {parameterName=UseLengthMask?}
  n41 = Constant3Vector()
  n42 = ScalarParameter() {defaultValue=850, parameterName=Radial Opacity Falloff}
  n43 = ScalarParameter() {defaultValue=0.75, parameterName=Radial Opacity Falloff Hardness}
  n44 = Power(
   Base=OneMinus(None=SphereMask(A=@n35, B=@n41, Radius=@n42, Hardness=@n43))
   Exp=ScalarParameter() {defaultValue=0.75, parameterName=Radial Opacity Falloff Power}
  )
  Multiply(A=LinearInterpolate(A=@n40, B=Multiply(A=@n40, B=@n44), Alpha=ScalarParameter() {parameterName=Fade With Sphere Mask}), B=ParticleColor().A)
  n45 = Multiply(
   A=ScalarParameter() {defaultValue=1250, parameterName=Width}
   B=Add(A=VertexColor().R, B=ScalarParameter() {defaultValue=0.2, parameterName=Ribbon Thickness Mask Multiplier}) {constB=0.785}
  )
  n46 = Multiply(A=DynamicParameter() {parameterIndex=1, defaultValue=<rgba 1 1 12 0>, paramNames [Random,Erode,CenterPower,Squeeze]}.Random, B=@n45)
  n47 = CollectionParameter() {collection=&/NexCollection, parameterName=WavePulseSplineWidthLerp}
  n48 = TransformPosition(None=WorldPosition().XYZ) {transformSourceType=TRANSFORMPOSSOURCE_World}
  n49 = Constant3Vector()
  n50 = LinearInterpolate(
   A=ScalarParameter() {group=Pulse Wave, parameterName=PulseWaveRadiusLow}
   B=ScalarParameter() {defaultValue=3500, group=Pulse Wave, parameterName=PulseWaveRadiusHigh}
   Alpha=CollectionParameter() {collection=&NexCollection, parameterName=WavePulseLerp}
  )
  n51 = SphereMask(A=@n48, B=@n49, Radius=@n50, Hardness=ScalarParameter() {defaultValue=0.8, group=Pulse Wave, parameterName=PulseWaveOuterHardness})
  n52 = Subtract(A=@n50, B=ScalarParameter() {defaultValue=512, group=Pulse Wave, parameterName=PulseWaveWidth})
  n53 = ScalarParameter() {defaultValue=0.1, group=Pulse Wave, parameterName=PulseWaveInnerHardness}
  n54 = OneMinus(None=Clamp(None=SphereMask(A=@n48, B=@n49, Radius=@n52, Hardness=@n53)))
  n55 = LinearInterpolate(
   A=@n46
   B=Multiply(A=@n46, B=LinearInterpolate(Alpha=@n47) {constA=1, constB=10}) {constB=12}
   Alpha=Multiply(A=Clamp(None=@n51), B=Power(Base=@n54) {constExponent=12})
  )
  n56 = MaterialFunctionCall(WidthBase=@n55, WidthTip=@n55) {materialFunction=&SplineThicken}
  n57 = Abs(None=DotProduct(A=CameraVectorWS(), B=@n56.Normal))
  Multiply(A=@n32, B=Clamp(None=Power(Base=@n57)))
  OneMinus(None=@n38)
  n58 = MaterialFunctionCall() {materialFunction=&CameraDirectionVector}
  n59 = Constant3Vector() {constant=<rgba 0 0 1 0>}
  n60 = MaterialFunctionCall(Float2=TextureCoordinate()) {materialFunction=&BreakOutFloat2Components}
  n61 = Frac(None=Multiply(A=Time()) {constB=0.25})
  n62 = Sine(None=Add(A=Multiply(A=@n60.G) {constB=5}, B=@n61))
  n63 = Sine(None=Add(A=Multiply(A=@n60.G) {constB=6}, B=@n61))
  n64 = Multiply(
   A=Multiply(A=CrossProduct(A=@n58.Result, B=@n59), B=Add(A=@n62, B=@n63))
   B=Multiply(
    A=ScalarParameter() {defaultValue=30, parameterName=WPO Gain}
    B=DynamicParameter() {defaultValue=<rgba 1 1 12 0>, paramNames [Random, WPO Gain Multiplier, Unassigned, Squeeze]}.WPO Gain Multiplier
   )
  )
  Reroute(None=ShadingPathSwitch(Default=Add(A=@n56.WorldPositionOffset, B=@n64), Forward=@n56.WorldPositionOffset, Mobile=@n56.WorldPositionOffset))
  StaticSwitchParameter(True=@n46, False=@n45) {parameterName=UseExtraNiagaraOptions?}
 ]
}
