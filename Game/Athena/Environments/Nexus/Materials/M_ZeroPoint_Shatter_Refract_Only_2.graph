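# Additive, unlit, two-sided refraction material for mesh particles, apparently for the Zero Point shatter effect in the Nexus environment.
# Vertex offset combines spline thickening, panned noise, Z-axis twist and gradient-weighted outward push, partly driven by dynamic parameters.
# Emissive blends two tinted, chroma-offset turbulence samples over a gradient color.
# Gradient, band, view-angle and depth fades shape opacity; refraction flattens on low quality and mobile, where emissive detail also drops at low-to-medium quality.
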Material /Game/Athena/Environments/Nexus/Materials/M_ZeroPoint_Shatter_Refract_Only_2
graph {
 properties {
  bUsedWithMeshParticles=true
  bUsedWithNiagaraMeshParticles=true
  blendMode=BLEND_Additive
  pixelDepthOffsetMode=PDOM_Legacy
  refractionMethod=RM_IndexOfRefraction
  shadingModel=MSM_Unlit
  twoSided=true
  usedShadingModels=Unlit
 }
 nodes [
  n1 = DynamicParameter() {defaultValue=<rgba 1 100 100 1>, paramNames [refraction, Thickness Start, Thickness End, Wpo Speed]}
  n2 = WorldPosition()
  n3 = MaterialFunctionCall(B=Multiply(A=@n1.Wpo Speed, B=Time())) {materialFunction=&AppendMany}
  n4 = DynamicParameter() {parameterIndex=1, paramNames [WPO Str, Wpo TIle X, Wpo TIle Y, Wpo TIle Z]}
  n5 = TextureSample(UVs=Divide(
    A=Add(A=@n2.XYZ, B=@n3.RGB)
    B=MaterialFunctionCall(R=@n4.Wpo TIle X, G=@n4.Wpo TIle X, B=@n4.Wpo TIle X) {materialFunction=&AppendMany}.RGB
   )) {texture=/Game/Effects/Explosion_Update/Textures/VT_ExplosionNoise}
  n6 = OneMinus(None=MaterialFunctionCall() {materialFunction=&LinearGradient}.VGradient)
  n7 = Multiply(A=Multiply(A=Subtract(A=@n5.RGB) {constB=0.5}, B=@n4.WPO Str), B=@n6)
  n8 = RotateAboutAxis(
   NormalizedRotationAxis=Constant3Vector() {constant=<rgba 0 0 1 0>}
   RotationAngle=Multiply(A=ScalarParameter() {defaultValue=1, parameterName=Twist amount}, B=@n6)
   PivotPoint=ActorPositionWS()
   Position=WorldPosition().XYZ
  )
  n9 = WorldPosition()
  n10 = ActorPositionWS()
  n11 = Normalize(VectorInput=Multiply(A=Subtract(A=@n9.XYZ, B=@n10), B=Constant3Vector() {constant=<rgba 1 1 0 0>}))
  n12 = Power(Base=NamedRerouteUsage(), Exp=ScalarParameter() {defaultValue=2, parameterName=outward grad Exp})
  Add(
   A=Add(
    A=MaterialFunctionCall(WidthBase=@n1.Thickness Start, WidthTip=@n1.Thickness Start) {materialFunction=&SplineThicken}.WorldPositionOffset
    B=Add(A=@n7, B=@n8)
   )
   B=Multiply(A=Multiply(A=@n11, B=@n12), B=ScalarParameter() {defaultValue=1000, parameterName=Outward amount})
  )
  DynamicParameter() {parameterIndex=3, paramNames [na,na,na,na]}
  n13 = ComponentMask(None=TextureCoordinate()) {g=true}
  n14 = Multiply(
   A=Saturate(None=Power(Base=@n13) {constExponent=12})
   B=VectorParameter() {parameterName=BackgroundColor, defaultValue=<rgba 1 0.0905 0.3951 0>}.RGB
  )
  n15 = Constant3Vector() {constant=<rgba 0 1 0.7274 0>}
  n16 = LinearInterpolate(Alpha=ComponentMask(None=TextureCoordinate() {coordinateIndex=1}) {g=true}) {constA=0.7}
  n17 = Multiply(A=ScalarParameter() {defaultValue=0.07, parameterName=ChromaOffset}, B=@n16) {constA=1}
  n18 = VectorParameter() {parameterName=Tile, defaultValue=<rgba 1.5 0.8 0 0>}
  n19 = Multiply(A=AppendVector(A=@n18.R, B=Multiply(A=@n18.G, B=@n16)), B=TextureCoordinate())
  n20 = Multiply(A=AppendVector(A=Constant() {r=1}, B=Add(B=@n17) {constA=1}), B=@n19)
  n21 = Multiply(A=ScalarParameter() {parameterName=ChromaOffsetAdd}, B=Constant2Vector() {g=1})
  n22 = Multiply(A=ScalarParameter() {defaultValue=0.1, parameterName=PanSpeed}, B=Add(A=Time(), B=ParticleRandom()))
  n23 = TextureSample(UVs=Panner(Coordinate=Add(A=@n20, B=@n21), Time=@n22) {speedX=0.04, speedY=0.12}) {texture=/Game/Effects/Fort_Effects/Textures/Plasma/T_Turbulence_Stretched}
  n24 = Constant3Vector() {constant=<rgba 1 0 0.1495 0>}
  n25 = TextureSample(UVs=Panner(Coordinate=@n19, Time=@n22) {speedX=0.04, speedY=0.12}) {texture=&T_Turbulence_Stretched}
  n26 = Add(A=@n14, B=LinearInterpolate(A=Multiply(A=@n15, B=@n23.G), B=Multiply(A=@n24, B=@n25.G)))
  n27 = QualitySwitch(Default=@n26, Low=Constant(), High=@n26, Medium=@n14)
  n28 = ParticleColor()
  MaterialFunctionCall(Emissive=Power(Base=Multiply(A=@n27, B=@n28.RGB), Exp=ScalarParameter() {defaultValue=1, parameterName=TextureDetail-Exponent})) {materialFunction=&MF_FX_EmissiveAdjustments}
  n29 = Constant() {r=0.5}
  n30 = ComponentMask(None=TextureCoordinate()) {g=true}
  n31 = Multiply(
   A=VertexColor().G
   B=SphereMask(A=@n30, B=@n29, Radius=@n29, Hardness=ScalarParameter() {defaultValue=0.25, parameterName=VerticalMaskHardness}) {hardnessPercent=64}
  )
  n32 = Multiply(
   A=MaterialFunctionCall("FadeSensitivity (optional)"=ScalarParameter() {defaultValue=1, parameterName=AngleFade}) {materialFunction=&FadeBasedOnViewAngle}.Result
   B=DepthFade(FadeDistance=ScalarParameter() {defaultValue=500, parameterName=FadeDistance})
  )
  n33 = Constant() {r=1}
  n34 = Multiply(A=@n28.A, B=Multiply(A=Power(Base=@n31) {constExponent=2.5}, B=FeatureLevelSwitch(Default=@n32, ES2=@n33, ES3_1=@n33, SM4_REMOVED=@n32)))
  n35 = DynamicParameter() {parameterIndex=2, paramNames [Gradient, Gradient contrast, na, na]}
  n36 = MaterialFunctionCall() {materialFunction=&LinearGradient}
  n37 = Saturate(None=MaterialFunctionCall(In=Add(A=@n35.Gradient, B=@n36.VGradient), Contrast=@n35.Gradient contrast) {materialFunction=&CheapContrast}.Result)
  n38 = Constant() {r=0.5}
  n39 = Constant() {r=0.65}
  n40 = StaticBool() {value=true}
  Multiply(
   A=Multiply(
    A=Multiply(A=@n34, B=@n37)
    B=MaterialFunctionCall!(Width=@n38, Sharpness=@n39, Offset, Direction Switch=@n40, Compare, Input Coordinates) {materialFunction=/Engine/Functions/Engine_MaterialFunctions03/Procedurals/GeneratedBand}.Result
   )
   B=MaterialFunctionCall(
    Fade Length=ScalarParameter() {defaultValue=250, parameterName="Camera Fade Length (S)"}
    Fade Offset=ScalarParameter() {defaultValue=100, parameterName="Camera Fade Offset (S)"}
   ) {materialFunction=&CameraDepthFade}.Result
  )
  NamedRerouteDeclaration(None=@n6) {name=V Grad, nodeColor=<rgba 0 1 0 1>}
  n41 = LinearInterpolate(B=Multiply(A=ScalarParameter() {defaultValue=1.05, parameterName=Refraction}, B=@n1.refraction), Alpha=Multiply(A=@n32, B=@n30)) {constA=1}
  n42 = Constant() {r=1}
  QualitySwitch(Default=FeatureLevelSwitch(Default=@n41, ES2=@n42, ES3_1=@n42, SM4_REMOVED=@n41), Low=@n42)
 ]
}
