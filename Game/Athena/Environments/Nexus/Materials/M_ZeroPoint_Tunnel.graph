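# Unlit, two-sided translucent mesh-particle material for the Zero Point tunnel effect, with depth testing disabled.
# Layers panning magnetic waves, curve-atlas-colored swirl noise, portal-space texture and cracked-glass cubemap reflections, keeping only swirls on Forward and Mobile.
# V gradients apply mist tint and a core glow whose falloff particles drive via dynamic parameter; particle alpha takes an optional gradient.
# A time-driven circular offset fades with distance from a local point, wobbling the tunnel.
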Material /Game/Athena/Environments/Nexus/Materials/M_ZeroPoint_Tunnel
graph {
 properties {
  bDisableDepthTest=true
  bUsedWithMeshParticles=true
  bUsedWithNiagaraMeshParticles=true
  blendMode=BLEND_Translucent
  pixelDepthOffsetMode=PDOM_Legacy
  shadingModel=MSM_Unlit
  twoSided=true
  usedShadingModels=Unlit
 }
 nodes [
  n1 = Multiply(A=Time(), B=ScalarParameter() {defaultValue=0.175, parameterName=RotationSpeed})
  n2 = Add(A=@n1, B=Sine(None=Multiply(A=@n1) {constB=0.45}))
  n3 = MaterialFunctionCall(R=Sine(None=@n2), B=Cosine(None=@n2)) {materialFunction=&AppendMany}
  n4 = ScalarParameter() {defaultValue=350, parameterName=Multiplier}
  n5 = WorldPosition() {worldPositionShaderOffset=WPT_ExcludeAllShaderOffsets}
  n6 = TransformPosition(None=VectorParameter() {parameterName=Offset, defaultValue=<rgba 0 50 0 1>}.RGB) {transformType=TRANSFORMPOSSOURCE_World}
  n7 = Saturate(None=Divide(A=Distance(A=@n5.XYZ, B=@n6), B=ScalarParameter() {defaultValue=150, parameterName=Rad}))
  LinearInterpolate(
   A=Constant3Vector()
   B=Transform(None=Multiply(A=@n3.RGB, B=@n4)) {transformSourceType=TRANSFORMSOURCE_Local}
   Alpha=Power(Base=OneMinus(None=@n7), Exp=ScalarParameter() {defaultValue=2, parameterName=LinearFalloffPower})
  )
  n8 = Multiply(
   A=TextureCoordinate()
   B=AppendVector(
    A=ScalarParameter() {defaultValue=1, group=UVs, parameterName=WaveU_Tile}
    B=ScalarParameter() {defaultValue=1, group=UVs, parameterName=WaveV_Tile}
   )
  )
  n9 = AppendVector(
   A=ScalarParameter() {defaultValue=0.15, group=UVs, parameterName=WaveU_Pan}
   B=ScalarParameter() {defaultValue=-0.75, group=UVs, parameterName=WaveV_Pan}
  )
  n10 = ScalarParameter() {defaultValue=0.025, group=UVs, parameterName=WaveDistortion}
  n11 = TextureSample(UVs=Multiply(A=TextureCoordinate()) {constB=2}) {samplerSource=SSM_Wrap_WorldGroupSettings, samplerType=SAMPLERTYPE_Normal, texture=/Game/Global/Textures/Default/Noise/T_SwirlNoise_N}
  n12 = ComponentMask(None=@n11.RGB) {g=true, r=true}
  n13 = TextureSampleParameter2D(UVs=Add(A=Panner(Coordinate=@n8, Speed=@n9) {speedX=0.15, speedY=-0.75}, B=Multiply(A=@n12, B=@n10) {constB=0.025})) {
   parameterName=WavesTexture
   samplerSource=SSM_Wrap_WorldGroupSettings
   texture=/Game/Athena/Prototype/Blueprints/Sneezy/Textures/T_Sneezy_MagneticWaves
  }
  n14 = Constant() {r=0.3}
  n15 = TextureCoordinate() {uTiling=2}
  n16 = TextureSample(UVs=Add(A=Panner(Coordinate=@n15) {speedX=0.15, speedY=-0.75}, B=Multiply() {constB=0.025})) {samplerSource=SSM_Wrap_WorldGroupSettings, texture=/Game/Athena/Prototype/Blueprints/Sneezy/Textures/T_Sneezy_EnergyWaves}
  n17 = Add(
   A=@n13.RGB
   B=Multiply(
    A=MaterialFunctionCall(Hue Shift Percentage=@n14, Texture=@n16.RGB) {materialFunction=&HueShift}.Result
    B=ScalarParameter() {defaultValue=1, group=Textures, parameterName=TwoWaveIntensity}
   )
  )
  n18 = Rotator(
   Coordinate=MaterialFunctionCall!(
    UV Channel=TextureCoordinate()
    U Freq Multiplier=ScalarParameter() {defaultValue=2, group=UVs, parameterName=UV U Mult}
    V Freq Multiplier=ScalarParameter() {defaultValue=5, group=UVs, parameterName=UV V Mult}
    Pan Speed U=ScalarParameter() {defaultValue=0.4, group=UVs, parameterName=UV U Pan Speed}
    Pan Speed V=ScalarParameter() {defaultValue=-3, group=UVs, parameterName=UV V Pan Speed}
   ) {materialFunction=/Game/Effects/Niagara/MasterMaterials/Functions/UVs/MF_FX_TextureUVPanner}.Result
   Time=Constant() {r=4.7125}
  ) {speed=1}
  n19 = ScalarParameter() {defaultValue=0.0745, parameterName=Distortion Multiplier}
  n20 = TextureSampleParameter2D(UVs=Add(A=@n18, B=Multiply(A=@n12, B=@n19) {constB=0.15})) {parameterName=MainTexture, samplerType=SAMPLERTYPE_LinearGrayscale, texture=/Game/Global/Textures/Default/Noise/T_SwirlNoise_N_height}
  n21 = CurveAtlasRowParameter(CurveTime=@n20.G) {atlas=&AthenaCurveAtlas01, curve=/Game/Athena/LinearColorCurves/Curve_GhostWeaponRamp, defaultValue=27}
  n22 = TextureSampleParameter2D(UVs=Multiply(A=@n18)) {parameterName=MainTexture, samplerType=SAMPLERTYPE_LinearGrayscale, texture=&T_SwirlNoise_N_height}
  n23 = CurveAtlasRowParameter(CurveTime=@n22.G) {atlas=&AthenaCurveAtlas01, curve=/Game/Athena/LinearColorCurves/LlamaHeroCurve, defaultValue=152}
  n24 = TextureCoordinate() {vTiling=0.5}
  n25 = Add(A=TextureSample(UVs=Panner(Coordinate=@n24) {speedY=0.8}) {samplerSource=SSM_Wrap_WorldGroupSettings, texture=/Game/Effects/Fort_Effects/Textures/Mask/T_Rand_Pan_Mask}.RGB) {constB=0.5}
  n26 = TextureCoordinate() {vTiling=0.15}
  n27 = Add(A=TextureSample(UVs=Panner(Coordinate=@n26) {speedY=0.5}) {samplerSource=SSM_Wrap_WorldGroupSettings, texture=&T_Rand_Pan_Mask}.RGB) {constB=0.25}
  n28 = Multiply(
   A=Multiply(A=Add(A=@n21, B=@n23), B=ScalarParameter() {defaultValue=0.75, group=Textures, parameterName=MainShapeIntensity}) {constB=0.75}
   B=Multiply(A=Add(A=@n25, B=@n27)) {constB=2.5}
  )
  n29 = TextureCoordinate()
  n30 = MaterialFunctionCall(
   Hue Shift Percentage=ScalarParameter() {group=Textures, parameterName=PortalSpaceHueShift}
   Texture=TextureSample(UVs=Panner(Coordinate=@n29) {speedX=0.065, speedY=-0.25}) {samplerSource=SSM_Wrap_WorldGroupSettings, texture=/Game/Effects/Fort_Effects/Textures/Tile/T_Portal_Space_D}.RGB
  ) {materialFunction=&HueShift}
  n31 = ScalarParameter() {defaultValue=4, group=Textures, parameterName=PortalSpaceIntensity}
  n32 = ScalarParameter() {parameterName=DEBUG}
  n33 = MaterialFunctionCall(
   UV Channel=TextureCoordinate()
   U Freq Multiplier=ScalarParameter() {defaultValue=4, group=UVs, parameterName=UV U Shatter Mult}
   V Freq Multiplier=ScalarParameter() {defaultValue=6, group=UVs, parameterName=UV V Shatter Mult}
   Pan Speed U=LinearInterpolate(A=ScalarParameter() {defaultValue=0.3, group=UVs, parameterName=UV U Shatter Pan Speed}, Alpha=@n32) {constB=0}
   Pan Speed V=LinearInterpolate(A=ScalarParameter() {defaultValue=-2, group=UVs, parameterName=UV V Shatter Pan Speed}, Alpha=@n32) {constB=0}
  ) {materialFunction=&MF_FX_TextureUVPanner}
  n34 = Add(A=TextureSample(UVs=@n33.Result) {samplerSource=SSM_Wrap_WorldGroupSettings, texture=/Game/Effects/Fort_Effects/Textures/Tile/T_Glass_Cracks_M}.R) {constB=0.3}
  n35 = TextureSample(UVs=Multiply(A=@n33.Result) {constB=4}) {
   samplerSource=SSM_Wrap_WorldGroupSettings
   samplerType=SAMPLERTYPE_Normal
   texture=/Game/Effects/Fort_Effects/Textures/Tile/T_glass_cracks_sharp_N
  }
  n36 = TextureSample(UVs=@n33.Result) {
   samplerSource=SSM_Wrap_WorldGroupSettings
   samplerType=SAMPLERTYPE_Normal
   texture=&T_glass_cracks_sharp_N
  }
  n37 = TextureSample(UVs=@n33.Result) {samplerSource=SSM_Wrap_WorldGroupSettings, texture=&T_Glass_Cracks_M}
  n38 = LinearInterpolate(
   A=Multiply(A=@n35.RGB, B=VectorParameter() {parameterName=Cube Normal Map Intensity, defaultValue=<rgba 8 8 1 1>}.RGB)
   B=Multiply(A=@n36.RGB, B=VectorParameter() {parameterName=Cube Normal Map 2 Intensity, defaultValue=<rgba 10 10 1 1>}.RGB)
   Alpha=Ceil(None=Subtract(A=@n37.R) {constB=0.1})
  )
  n39 = TextureSampleParameterCube!(UVs=ReflectionVectorWS(CustomWorldNormal=Transform(None=@n38)), Apply View MipBias) {
   parameterName=Reflection Cube Map
   samplerSource=SSM_Wrap_WorldGroupSettings
   texture=/Game/Packages/Fortress_Sky/SkyDome/MaterialInstances/SkyDome_Apollo/Cubemap_Apollo_Night
  }
  n40 = TextureSample(UVs=@n33.Result) {constMipValue=3, samplerSource=SSM_Wrap_WorldGroupSettings, texture=/Game/Global/Textures/Default/Noise/3xPerlin_M}
  n41 = ScalarParameter() {defaultValue=0.2, parameterName=Crystal Rainbow Desat}
  n42 = ComponentMask(None=@n39.RGB) {b=true}
  n43 = Multiply(A=Multiply(A=Desaturation(None=@n40.RGB, Fraction=@n41)) {constB=3}, B=Multiply(A=Power(Base=@n42) {constExponent=5}) {constB=40})
  n44 = Add(A=Multiply(A=Saturate(None=@n34), B=Add(A=@n39.RGB, B=@n43)), B=Add(A=Add(A=@n28, B=@n17), B=Multiply(A=@n30.Result, B=@n31) {constB=4}))
  n45 = ShadingPathSwitch(Default=@n44, Deferred=@n44, Forward=@n28, Mobile=@n28)
  n46 = MaterialFunctionCall() {materialFunction=&LinearGradient}
  n47 = ScalarParameter() {parameterName=GradientPowerInnerShadow}
  n48 = Power(Base=OneMinus(None=@n46.VGradient), Exp=ScalarParameter() {defaultValue=1.25, parameterName=GradientPowerOuterShadow})
  n49 = Multiply(
   A=LinearInterpolate(A=VectorParameter() {parameterName=Tunnel Mist Color, defaultValue=<rgba 0 0.0133 0.026 0>}.RGB, Alpha=Power(Base=@n46.VGradient, Exp=@n47))
   B=Multiply(A=@n48) {constB=2}
  )
  n50 = Power(Base=@n48, Exp=DynamicParameter() {defaultValue=<rgba 4 1 1 1>, paramNames [Core Glow Power, Param2, Param3, Param4]}.Core Glow Power)
  n51 = VectorParameter() {parameterName=CoreColor, defaultValue=<rgba 0 37.7137 50 0>}
  MaterialFunctionCall(Emissive=Add(A=Multiply(A=@n45, B=@n49), B=Multiply(A=@n50, B=@n51.RGB)) {constB=0}) {materialFunction=&MF_FX_EmissiveAdjustments}
  n52 = ParticleColor()
  n53 = MaterialFunctionCall() {materialFunction=&LinearGradient}
  StaticSwitchParameter(True=Multiply(A=@n52.A, B=Multiply(A=@n53.VGradient, B=@n53.VGradient)), False=@n52.A) {parameterName=UseCenterGradient?}
 ]
}
