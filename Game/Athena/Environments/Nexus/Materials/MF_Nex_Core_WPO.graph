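# Material function for the Nexus core that computes a world position offset spreading mesh pieces apart, apparently as a shatter effect.
# Vertex color, remapped to signed local-space vectors and transformed, gives the direction, masked by UV1's green channel.
# Progress blends MinSpread to MaxSpread, optionally modulated by layered time-driven sines through Lerp Shatter Pulse, Shake and Time Multiplier inputs.
# It outputs the offset plus a saturated spread-based mask.
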MaterialFunction /Game/Athena/Environments/Nexus/Materials/MF_Nex_Core_WPO
graph {nodes [
  n1 = ComponentMask(None=TextureCoordinate() {coordinateIndex=1}) {g=true}
  n2 = VertexColor()
  n3 = Transform(None=Multiply(A=Subtract(A=@n2) {constB=0.5}) {constB=2}) {transformSourceType=TRANSFORMSOURCE_Local}
  n4 = FunctionInput(Preview=ScalarParameter() {parameterName=Progress}) {bUsePreviewValueAsDefault=true, inputName=Progress, inputType=FunctionInput_Scalar}
  n5 = FunctionInput() {bUsePreviewValueAsDefault=true, inputName=Time Multiplier, inputType=FunctionInput_Scalar, previewValue=<1 0 0 1>}
  n6 = Multiply(A=Time(), B=@n5)
  n7 = Multiply(A=@n6) {constB=0.5}
  n8 = LinearInterpolate(
   A=ConstantBiasScale(None=Sine(None=@n6) {period=0.25})
   B=ConstantBiasScale(None=Sine(None=@n6))
   Alpha=ConstantBiasScale(None=Sine(None=@n6) {period=0.8})
  )
  n9 = Sine(None=@n6) {period=2}
  n10 = LinearInterpolate(
   A=ConstantBiasScale(None=Sine(None=@n7) {period=0.25})
   B=LinearInterpolate(B=Multiply(A=@n8) {constB=2.5}, Alpha=FunctionInput() {bUsePreviewValueAsDefault=true, inputName=Shake, inputType=FunctionInput_Scalar})
   Alpha=Clamp(None=ConstantBiasScale(None=@n9) {bias=2})
  ) {constB=0}
  n11 = LinearInterpolate(
   A=ScalarParameter() {defaultValue=200, parameterName=MinSpread}
   B=ScalarParameter() {defaultValue=1000, parameterName=MaxSpread}
   Alpha=LinearInterpolate(
    A=@n4
    B=Multiply(A=@n4, B=@n10)
    Alpha=FunctionInput() {bUsePreviewValueAsDefault=true, inputName=Lerp Shatter Pulse, inputType=FunctionInput_Scalar}
   )
  )
  FunctionOutput(None=Multiply(A=Multiply(A=@n1, B=@n3), B=@n11)) {outputName=WPOSpread}
  n12 = Divide(A=@n11) {constB=400}
  FunctionOutput(None=Multiply(A=@n1, B=Saturate(None=@n12))) {outputName=Mask}
 ]}
