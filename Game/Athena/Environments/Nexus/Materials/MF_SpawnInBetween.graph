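# Material function for a Nexus environment that adjusts World Position Offset to pull vertices horizontally toward a world center point.
# It subtracts the offset world position from the WorldCenter vector parameter, scales the result by the Animation value from MPC_Butterfly, and masks it to X and Y.
# That result is added to the incoming WPO, so the shared collection value drives the collapse or spread.
# A SpawnAnim scalar parameter is present but unconnected.
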MaterialFunction /Game/Athena/Environments/Nexus/Materials/MF_SpawnInBetween
graph {nodes [
  n1 = FunctionInput(Preview=Constant3Vector()) {bUsePreviewValueAsDefault=true, inputName=WPO}
  n2 = WorldPosition()
  n3 = Multiply(
   A=Subtract(A=VectorParameter() {parameterName=WorldCenter, defaultValue=<rgba 0 0 248678.5312 1>}.RGB, B=Add(A=@n1, B=@n2.XYZ))
   B=CollectionParameter() {collection=&/MPC_Butterfly, parameterName=Animation}
  )
  n4 = Constant3Vector() {constant=<rgba 1 1 0 0>}
  FunctionOutput(None=Add(A=@n1, B=Multiply(A=@n3, B=@n4)))
  ScalarParameter() {parameterName=SpawnAnim}
 ]}
